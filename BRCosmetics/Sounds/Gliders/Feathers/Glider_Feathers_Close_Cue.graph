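# Sound cue for the Feathers glider cosmetic's close sound in Battle Royale.
# A single wave player plays the Glider_Feathers_Close02 wave through a Modulator node left at its default settings.
# The glider open/close attenuation and concurrency assets, the glider landing sound class and a cosmetic glider-close property sheet govern its playback.
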SoundCue /BRCosmetics/Sounds/Gliders/Feathers/Glider_Feathers_Close_Cue
root=Modulator_0
nodes [
 {id=Modulator_0, children [WavePlayer_0]}
 {id=WavePlayer_0, properties {soundWaveAssetPtr=&/Glider_Feathers_Close02}}
]
properties {
 attenuationSettings=/Game/Sounds/Attenuation/Fort_Player_Glider_OpenClose_Athena_Att
 audioPropertiesSheet=/BRCosmetics/Sounds/Settings/PropertySheets/Glider/APS_Cosmetic_Glider_Close
 maxDistance=10000
 soundClassObject=&Fort_Gliders_Land
 volumeMultiplier=0.9
 concurrencySet [&Athena_GliderOpenClose_Conc]
}
