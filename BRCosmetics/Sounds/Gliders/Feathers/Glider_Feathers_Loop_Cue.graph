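# Looping sound cue for the Feathers glider cosmetic, spatialized with player glider thrust attenuation.
# A mixer layers a continuously looping ambient bed with a quieter branch of intermittent vocal one-shots.
# The vocal branch repeats through a Looping node, waits a random few seconds, then picks one of five clips by weighted random.
# Modulators vary volume on both layers and raise the pitch of two of the less frequent vocal clips.
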SoundCue /BRCosmetics/Sounds/Gliders/Feathers/Glider_Feathers_Loop_Cue
root=Mixer_0
nodes [
 {id=Delay_0, children [Modulator_0], properties {delayMax=8, delayMin=3}}
 {id=Looping_0, children [Delay_0]}
 {id=Mixer_0, children [Modulator_1,Looping_0], properties {inputVolume [0.9,0.6]}}
 {id=Modulator_0, children [Random_1], properties {volumeMax=1.3, volumeMin=1.2}}
 {id=Modulator_1, children [WavePlayer_1], properties {pitchMax=1, pitchMin=1, volumeMax=0.9, volumeMin=0.8}}
 {id=Modulator_3, children [WavePlayer_6], properties {pitchMax=1.2, pitchMin=1.2, volumeMax=1.1, volumeMin=1.1}}
 {id=Modulator_4, children [WavePlayer_7], properties {pitchMax=1.2, pitchMin=1.2, volumeMax=1.1, volumeMin=1.1}}
 {id=Random_1, children [Modulator_3, Modulator_4, WavePlayer_9, WavePlayer_10, WavePlayer_11], properties {weights [1,1,2,3,3]}}
 {id=WavePlayer_1, properties {bLooping=true, soundWaveAssetPtr=&/Glider_Feathers_AmbientLoop_01}}
 {id=WavePlayer_10, properties {soundWaveAssetPtr=&/Glider_Feathers_Vocal_Random09}}
 {id=WavePlayer_11, properties {soundWaveAssetPtr=&/Glider_Feathers_Vocal_Random10}}
 {id=WavePlayer_6, properties {soundWaveAssetPtr=&/Glider_Feathers_Vocal_Random03}}
 {id=WavePlayer_7, properties {soundWaveAssetPtr=&/Glider_Feathers_Vocal_Random04}}
 {id=WavePlayer_9, properties {soundWaveAssetPtr=&/Glider_Feathers_Vocal_Random08}}
]
properties {
 attenuationSettings=&Fort_Player_Glider_Thrust_Athena_Att
 audioPropertiesSheet=&APS_Cosmetic_Glider_Loop
 maxDistance=12500
 soundClassObject=&Fort_Gliders_Land
 volumeMultiplier=0.45
}
