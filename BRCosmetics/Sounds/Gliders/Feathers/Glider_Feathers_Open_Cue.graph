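# Sound cue for opening the Feathers glider, layering a deploy sound with a delayed vocal.
# A mixer plays the open wave immediately, alongside a quieter branch that waits a fixed delay before picking one of four vocal waves.
# The weighted random pick allows repeats, favors the first vocal and makes the fourth very rare, with modulators giving some variants fixed pitch or volume offsets.
# It uses shared glider open/close attenuation and concurrency settings.
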SoundCue /BRCosmetics/Sounds/Gliders/Feathers/Glider_Feathers_Open_Cue
root=Mixer_0
nodes [
 {id=Delay_0, children [Random_0], properties {delayMax=1.82, delayMin=1.82}}
 {id=Mixer_0, children [WavePlayer_11, Modulator_0], properties {inputVolume [1.1,0.6]}}
 {id=Modulator_0, children [Delay_0], properties {pitchMax=1}}
 {id=Modulator_2, children [WavePlayer_7], properties {pitchMax=1.1, pitchMin=1.1, volumeMax=1.1, volumeMin=1.1}}
 {id=Modulator_3, children [WavePlayer_8], properties {pitchMax=1, pitchMin=1, volumeMax=1.1, volumeMin=1.1}}
 {id=Modulator_4, children [WavePlayer_10], properties {pitchMax=0.9, pitchMin=0.9}}
 {
  id=Random_0
  children [WavePlayer_6,Modulator_2,Modulator_3,Modulator_4]
  properties {bRandomizeWithoutReplacement=false, weights [2,1,1,0.01]}
 }
 {id=WavePlayer_10, properties {soundWaveAssetPtr=&/Glider_Feathers_Vocal_Open04}}
 {id=WavePlayer_11, properties {soundWaveAssetPtr=&/Glider_Feathers_Open02}}
 {id=WavePlayer_6, properties {soundWaveAssetPtr=&/Glider_Feathers_Vocal_Open01}}
 {id=WavePlayer_7, properties {soundWaveAssetPtr=&/Glider_Feathers_Vocal_Open02}}
 {id=WavePlayer_8, properties {soundWaveAssetPtr=&/Glider_Feathers_Vocal_Open03}}
]
properties {
 attenuationSettings=/Game/Sounds/Attenuation/Fort_Player_Glider_OpenClose_Athena_Att
 audioPropertiesSheet=/BRCosmetics/Sounds/Settings/PropertySheets/Glider/APS_Cosmetic_Glider_Open
 maxDistance=10000
 soundClassObject=&Fort_Gliders_Land
 volumeMultiplier=0.86
 concurrencySet [&Athena_GliderOpenClose_Conc]
}
