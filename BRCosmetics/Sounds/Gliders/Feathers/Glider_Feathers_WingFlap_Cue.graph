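# Sound cue for the wing flap of the Feathers glider cosmetic.
# A Random node picks one of five equally weighted flap recordings, which then passes through a Modulator for pitch and volume variation.
# It uses the player glider thrust attenuation and the glider landing sound class at reduced volume, with a concurrency override that caps simultaneous instances and applies volume scaling.
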SoundCue /BRCosmetics/Sounds/Gliders/Feathers/Glider_Feathers_WingFlap_Cue
root=Modulator_0
nodes [
 {id=Modulator_0, children [Random_0]}
 {id=Random_0, children [WavePlayer_1,WavePlayer_2,WavePlayer_3,WavePlayer_6,WavePlayer_7], properties {weights [1,1,1,1,1]}}
 {id=WavePlayer_1, properties {soundWaveAssetPtr=&/Glider_Feathers_WingFlap01}}
 {id=WavePlayer_2, properties {soundWaveAssetPtr=&/Glider_Feathers_WingFlap02}}
 {id=WavePlayer_3, properties {soundWaveAssetPtr=&/Glider_Feathers_WingFlap03}}
 {id=WavePlayer_6, properties {soundWaveAssetPtr=&/Glider_Feathers_WingFlap04}}
 {id=WavePlayer_7, properties {soundWaveAssetPtr=&/Glider_Feathers_WingFlap05}}
]
properties {
 attenuationSettings=&Fort_Player_Glider_Thrust_Athena_Att
 bOverrideConcurrency=true
 maxDistance=12500
 soundClassObject=&Fort_Gliders_Land
 volumeMultiplier=0.25
 concurrencyOverrides {maxCount=5, volumeScale=0.9}
}
